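# Translucent UI material for crafting indicator art in the Athena inventory HUD, apparently built for Sequencer animation.
# Static switches select a radial glow, two-channel sparks from scaled, thresholded noise, or a rotated, scaled, offset indicator texture with selectable channel.
# Start/Stop parameter pairs blended by Anim scalars drive spark scale and threshold, plus indicator rotation, scale and intensity.
# The resulting mask blends two colors and, times Final_Alpha, sets opacity.
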Material /Game/Athena/HUD/Inventory/Art/M_UI_CraftingIndicato_SequencerAnim
graph {
 properties {blendMode=BLEND_Translucent, materialDomain=MD_UI, pixelDepthOffsetMode=PDOM_Legacy, usedShadingModels=Default Lit}
 nodes [
  n1 = MaterialFunctionCall() {materialFunction=&GetUserInterfaceUV}
  n2 = MaterialFunctionCall(Float2=@n1.Pixel Size) {materialFunction=&BreakOutFloat2Components}
  n3 = Divide(A=@n2.R, B=@n2.R)
  n4 = Constant() {r=1}
  n5 = OneMinus(None=@n3)
  n6 = Add(A=Multiply(A=@n1.Normalized UV, B=AppendVector(A=@n3, B=@n4)), B=AppendVector(A=Multiply(A=@n5) {constB=0.5}, B=Constant()))
  n7 = MaterialFunctionCall(
   UVs=@n6
   Texture Scale=LinearInterpolate(
    A=ScalarParameter() {defaultValue=1, parameterName=Noise1Scale_Start}
    B=ScalarParameter() {defaultValue=1, parameterName=Noise1Scale_Stop}
    Alpha=ScalarParameter() {parameterName=Noise1Scale_Anim}
   )
  ) {materialFunction=&ScaleUVsByCenter}
  n8 = ScalarParameter() {parameterName=Noise1Subtract_Start}
  n9 = ScalarParameter() {parameterName=Noise1Subtract_Stop}
  n10 = ScalarParameter() {parameterName=Noise1Subtract_Anim}
  n11 = Multiply(
   A=Subtract(
    A=TextureSample(UVs=@n7.UVs) {texture=&/T_UI_Crafting_Sparks_128x}.R
    B=LinearInterpolate(A=@n8, B=@n9, Alpha=@n10)
   )
   B=ScalarParameter() {defaultValue=1, parameterName=Lift_RedChannel_Sparks}
  ) {constB=3}
  n12 = MaterialFunctionCall(
   UVs=@n6
   Texture Scale=LinearInterpolate(
    A=ScalarParameter() {defaultValue=1, parameterName=Noise2Scale_Start}
    B=ScalarParameter() {defaultValue=1, parameterName=Noise2Scale_Stop}
    Alpha=ScalarParameter() {parameterName=Noise2Scale_Anim}
   )
  ) {materialFunction=&ScaleUVsByCenter}
  n13 = ScalarParameter() {parameterName=Noise2Subtract_Start}
  n14 = ScalarParameter() {parameterName=Noise2Subtract_Stop}
  n15 = ScalarParameter() {parameterName=Noise2Subtract_Anim}
  n16 = Multiply(
   A=Subtract(
    A=TextureSample(UVs=@n12.UVs) {texture=&T_UI_Crafting_Sparks_128x}.G
    B=LinearInterpolate(A=@n13, B=@n14, Alpha=@n15)
   )
   B=ScalarParameter() {defaultValue=1, parameterName=Lift_GreenChannel_Sparks}
  ) {constB=3}
  n17 = Add(
   A=Multiply(A=Saturate(None=@n11), B=ScalarParameter() {defaultValue=1, parameterName=RedChannelSparksAmount})
   B=Multiply(A=Saturate(None=@n16), B=ScalarParameter() {defaultValue=1, parameterName=GreenChannelSparksAmount})
  )
  n18 = ScalarParameter() {defaultValue=0.5, parameterName=RotationCenter_X}
  n19 = ScalarParameter() {defaultValue=0.5, parameterName=RotationCenter_Y}
  n20 = ScalarParameter() {parameterName=RotationAngle_Start}
  n21 = ScalarParameter() {defaultValue=1, parameterName=RotationAngle_Stop}
  n22 = ScalarParameter() {parameterName=RotationAngle_Anim}
  n23 = MaterialFunctionCall(
   UVs=MaterialFunctionCall(UVs=@n6, Rotation Center=AppendVector(A=@n18, B=@n19), "Rotation Angle (0-1)"=LinearInterpolate(A=@n20, B=@n21, Alpha=@n22)) {materialFunction=&CustomRotator}.Rotated Values
   Texture Scale=LinearInterpolate(
    A=ScalarParameter() {defaultValue=1, parameterName=Scale_Start}
    B=ScalarParameter() {defaultValue=1, parameterName=Scale_Stop}
    Alpha=ScalarParameter() {defaultValue=1, parameterName=Scale_Anim}
   )
  ) {materialFunction=&ScaleUVsByCenter}
  n24 = AppendVector(A=ScalarParameter() {parameterName=Offset_X}, B=ScalarParameter() {parameterName=Offset_Y})
  n25 = StaticComponentMaskParameter(None=TextureSample(UVs=Add(A=@n23.UVs, B=@n24)) {texture=&/T_UI_CraftingIndicator_128x}.RGB) {defaultR=true, parameterName=Channel}
  n26 = LinearInterpolate(
   A=ScalarParameter() {defaultValue=1, parameterName=Alpha_Intensity_Start}
   B=ScalarParameter() {defaultValue=1, parameterName=Alpha_Intensity_Stop}
   Alpha=ScalarParameter() {parameterName=Alpha_Intensity_Anim}
  )
  n27 = AppendVector(A=ScalarParameter() {parameterName=Glow_Shift_X}, B=ScalarParameter() {parameterName=Glow_Shift_Y})
  n28 = StaticSwitchParameter(
   True=MaterialFunctionCall(
    UVs=Add(A=@n6, B=@n27)
    Radius=ScalarParameter() {defaultValue=0.5, parameterName=Glow_Radius}
    Density=ScalarParameter() {defaultValue=2.333, parameterName=Glow_Density}
   ) {materialFunction=&RadialGradientExponential}.RadialGradientExponential
   False=StaticSwitchParameter(True=Saturate(None=@n17), False=Multiply(A=@n25, B=@n26)) {parameterName=IsSparks}
  ) {parameterName=IsGlow}
  LinearInterpolate(
   A=VectorParameter() {parameterName=Color_2, defaultValue=<rgba 1 0 0 1>}.RGB
   B=VectorParameter() {parameterName=Color_1, defaultValue=<rgba 0 1 0 1>}.RGB
   Alpha=@n28
  )
  Saturate(None=Multiply(A=@n28, B=ScalarParameter() {defaultValue=1, parameterName=Final_Alpha}))
 ]
}
