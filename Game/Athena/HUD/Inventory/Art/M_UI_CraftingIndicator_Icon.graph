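# UI-domain translucent material for a crafting indicator icon, built from hammer (or axe) and flame masks in one shared channel-packed texture.
# Each mask samples the texture with its own parameterized center scale and XY offset, reading red for the hammer and green for the flames.
# The red mask blends flame color to hammer color, while the saturated sum of both masks forms the combined icon silhouette.
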Material /Game/Athena/HUD/Inventory/Art/M_UI_CraftingIndicator_Icon
graph {
 properties {blendMode=BLEND_Translucent, materialDomain=MD_UI, pixelDepthOffsetMode=PDOM_Legacy, usedShadingModels=Default Lit}
 nodes [
  n1 = ScalarParameter() {defaultValue=1, parameterName=Axe_Scale}
  n2 = TextureCoordinate()
  n3 = ScalarParameter() {parameterName=Axe_X_Position}
  n4 = ScalarParameter() {parameterName=Axe_Y_Position}
  n5 = TextureSample(UVs=Add(
    A=MaterialFunctionCall(UVs=@n2, Texture Scale=@n1) {materialFunction=&ScaleUVsByCenter}.UVs
    B=AppendVector(A=@n3, B=@n4)
   )) {texture=&/T_UI_CraftingIndicator_128x}
  LinearInterpolate(
   A=VectorParameter() {parameterName=Flame_Color, defaultValue=<rgba 1 0 1 1>}.RGB
   B=VectorParameter() {parameterName=Hammer_Color, defaultValue=<rgba 0 1 0 1>}.RGB
   Alpha=@n5.R
  )
  n6 = Add(
   A=MaterialFunctionCall(UVs=@n2, Texture Scale=ScalarParameter() {defaultValue=1, parameterName=Flames_Scale}) {materialFunction=&ScaleUVsByCenter}.UVs
   B=AppendVector(A=ScalarParameter() {parameterName=Flames_X_Position}, B=ScalarParameter() {parameterName=Flames_Y_Position})
  )
  Saturate(None=Add(A=@n5.R, B=TextureSample(UVs=@n6) {texture=&T_UI_CraftingIndicator_128x}.G))
 ]
}
