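# Translucent UI master material for the HUD crafting indicator, drawing an animated hammer swing with sparks, flame and glow.
# Time, offset by a randomizer parameter, loops through SmoothCurve to drive hammer rotation, spark and flame scaling, and an impact pulse.
# Packed texture channels supply the hammer and alternating spark and flame variants, switched by time-based comparisons.
# The pulse tints and fades flame and radial glow, with all layers combined into color and opacity.
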Material /Game/Athena/HUD/Inventory/Art/M_UI_CraftingIndicator_Master
graph {
 properties {blendMode=BLEND_Translucent, materialDomain=MD_UI, pixelDepthOffsetMode=PDOM_Legacy, usedShadingModels=Default Lit}
 nodes [
  n1 = Add(A=Time(), B=ScalarParameter() {parameterName=GlobalTimeOffsetRandomizer})
  n2 = ScalarParameter() {defaultValue=1.75, parameterName=SwingFrequency}
  n3 = Fmod(A=Multiply(A=@n1, B=ScalarParameter() {defaultValue=1.5, parameterName=GlobalTimeMultiplier}) {constB=1.5}, B=@n2)
  n4 = ScalarParameter() {defaultValue=3, parameterName=SwingBottomTangent}
  n5 = ScalarParameter() {defaultValue=4, parameterName=SwingTopTangent}
  n6 = Saturate(None=MaterialFunctionCall!(
    X=MaterialFunctionCall(X=@n3, Tangent 0=@n4, Tangent 1=@n5) {materialFunction=/Engine/Functions/Engine_MaterialFunctions02/Gradients/SmoothCurve}.Result
    Tangent 0
    Tangent 1
   ) {materialFunction=&SmoothCurve}.Result)
  n7 = OneMinus(None=@n3)
  n8 = Saturate(None=DotProduct(
    A=OneMinus(None=@n6)
    B=MaterialFunctionCall(X=@n7) {materialFunction=&SmoothCurve}.Result
   ))
  n9 = MaterialFunctionCall() {materialFunction=&GetUserInterfaceUV}
  n10 = MaterialFunctionCall(Float2=@n9.Pixel Size) {materialFunction=&BreakOutFloat2Components}
  n11 = Divide(A=@n10.R, B=@n10.R)
  n12 = Constant() {r=1}
  n13 = OneMinus(None=@n11)
  n14 = Add(A=Multiply(A=@n9.Normalized UV, B=AppendVector(A=@n11, B=@n12)), B=AppendVector(A=Multiply(A=@n13) {constB=0.5}, B=Constant()))
  n15 = MaterialFunctionCall(
   UVs=@n14
   Rotation Center=AppendVector(
    A=ScalarParameter() {defaultValue=0.666, parameterName=HammerRotationAngleX}
    B=ScalarParameter() {defaultValue=0.333, parameterName=HammerRotationAngleY}
   )
   "Rotation Angle (0-1)"=LinearInterpolate(A=ScalarParameter() {parameterName=Rotation_Sart}, B=ScalarParameter() {defaultValue=-0.075, parameterName=Rotation_Stop}, Alpha=@n6) {constA=-0.05, constB=0}
  ) {materialFunction=&CustomRotator}
  n16 = LinearInterpolate(
   A=ScalarParameter() {defaultValue=0.8, group=ItemIcon, parameterName=Icon_Scale_Start}
   B=ScalarParameter() {defaultValue=1, group=ItemIcon, parameterName=Icon_Scale_Stop}
   Alpha=ScalarParameter() {group=Animate, parameterName=Animate_ItemIcon_Scale}
  )
  n17 = LinearInterpolate(B=ScalarParameter() {defaultValue=-0.125, parameterName=AnimateX_Offset_Amount}, Alpha=ScalarParameter() {parameterName=AnimateX_Offset})
  n18 = AppendVector(A=@n17, B=Constant())
  n19 = TextureSample(UVs=Add(
    A=MaterialFunctionCall(UVs=@n15.Rotated Values, Texture Scale=@n16) {materialFunction=&ScaleUVsByCenter}.UVs
    B=@n18
   )) {texture=&/T_UI_CraftingIndicator_128x}
  n20 = ScalarParameter() {defaultValue=0.95, parameterName=ScaleSparksStartR}
  n21 = ScalarParameter() {defaultValue=1, parameterName=ScaleSparksStopR}
  n22 = Saturate(None=Power(Base=@n3, Exp=ScalarParameter() {defaultValue=1.22, parameterName=ScaleSparksRate}))
  n23 = Add(A=@n14, B=AppendVector(A=Constant(), B=ScalarParameter() {parameterName=ShiftSparksY}))
  n24 = Add(
   A=MaterialFunctionCall(UVs=@n23, Texture Scale=LinearInterpolate(A=@n20, B=@n21, Alpha=@n22) {constA=0.95}) {materialFunction=&ScaleUVsByCenter}.UVs
   B=@n18
  )
  n25 = Saturate(None=Multiply(
    A=TextureSample(UVs=@n24) {texture=&/T_UI_Crafting_Sparks_128x}.R
    B=ScalarParameter() {defaultValue=1, parameterName=Lift_RedChannel_Sparks}
   ) {constB=3})
  n26 = ScalarParameter() {defaultValue=0.95, parameterName=ScaleSparksStartG}
  n27 = ScalarParameter() {defaultValue=1, parameterName=ScaleSparksStopG}
  n28 = Add(
   A=MaterialFunctionCall(UVs=@n23, Texture Scale=LinearInterpolate(A=@n26, B=@n27, Alpha=@n22) {constA=0.95}) {materialFunction=&ScaleUVsByCenter}.UVs
   B=@n18
  )
  n29 = Saturate(None=Multiply(
    A=TextureSample(UVs=@n28) {texture=&T_UI_Crafting_Sparks_128x}.G
    B=ScalarParameter() {defaultValue=1, parameterName=Lift_GreenChannel_Sparks}
   ) {constB=3})
  n30 = ScalarParameter() {defaultValue=1.75, parameterName=SwingFrequency}
  n31 = Fmod(A=Multiply(A=Time(), B=ScalarParameter() {defaultValue=1.5, parameterName=GlobalTimeMultiplier}) {constB=1.5}, B=Multiply(A=@n30) {constB=2})
  n32 = Constant() {r=1}
  n33 = Constant()
  n34 = LinearInterpolate(
   A=Saturate(None=Multiply(A=@n7, B=@n25))
   B=Saturate(None=Multiply(A=@n7, B=@n29))
   Alpha=Saturate(None=If(A=@n31, B=@n30, "A > B"=@n32, "A < B"=@n33))
  )
  LinearInterpolate(
   A=LinearInterpolate(
    A=VectorParameter() {parameterName=Flame_Color_Stop, defaultValue=<rgba 1 0 1 1>}.RGB
    B=VectorParameter() {parameterName=Flame_Color_Start, defaultValue=<rgba 0 0 1 1>}.RGB
    Alpha=@n8
   )
   B=VectorParameter() {parameterName=Hammer_Color, defaultValue=<rgba 0 1 0 1>}.RGB
   Alpha=Saturate(None=Add(A=@n19.R, B=@n34))
  )
  n35 = Power(Base=@n7, Exp=ScalarParameter() {defaultValue=3, parameterName=ScaleFlameRate})
  n36 = MaterialFunctionCall!(
   Texture Scale=LinearInterpolate(
    A=ScalarParameter() {defaultValue=0.95, parameterName=ScaleFlameStart}
    B=ScalarParameter() {defaultValue=1, parameterName=ScaleFlameStop}
    Alpha=Saturate(None=@n35)
   ) {constA=0.95}
   UVs=@n14
   Offset XY Texturesize Z=VectorParameter() {parameterName=HandleFlameScaleLocation, defaultValue=<rgba 1 1 128 1>}.RGB
  ) {materialFunction=/Engine/Functions/Engine_MaterialFunctions01/Texturing/ScaleUVsAroundPoint}
  n37 = Add(A=@n36.UVs, B=@n18)
  n38 = Add(A=@n36.UVs, B=@n18)
  n39 = Multiply(
   A=Add(A=@n1, B=ScalarParameter() {defaultValue=0.5, parameterName=TimeOffset})
   B=Multiply(A=ScalarParameter() {defaultValue=1.5, parameterName=GlobalTimeMultiplier}) {constB=2}
  ) {constB=1.5}
  n40 = Fmod(A=@n39, B=Constant() {r=1})
  n41 = Constant() {r=-1}
  n42 = If(A=Fmod(A=@n39, B=Multiply(A=@n2) {constB=2}), B=@n2, "A > B"=Constant(), "A == B"=@n41, "A < B"=@n41)
  n43 = LinearInterpolate(
   A=Multiply(A=TextureSample(UVs=@n37) {texture=&T_UI_CraftingIndicator_128x}.B, B=@n8)
   B=Multiply(A=TextureSample(UVs=@n38) {texture=&T_UI_CraftingIndicator_128x}.G, B=@n8)
   Alpha=If(A=Add(A=@n40, B=@n42) {constB=-1}, B=Constant(), "A > B"=Constant() {r=1}, "A < B"=Constant())
  )
  n44 = ScalarParameter() {defaultValue=0.5, parameterName=gradientcenter_x}
  n45 = MaterialFunctionCall(
   UVs=TextureCoordinate()
   CenterPosition=AppendVector(A=Add(A=@n44, B=@n17), B=ScalarParameter() {defaultValue=1, parameterName=gradientcenter_y})
   Radius=LinearInterpolate(
    A=ScalarParameter() {defaultValue=0.5, parameterName=gradientRadius_Start}
    B=ScalarParameter() {defaultValue=0.5, parameterName=gradientRadius_Stop}
    Alpha=OneMinus(None=@n6)
   )
  ) {materialFunction=&RadialGradientExponential}
  n46 = Add(
   A=Saturate(None=Add(A=@n19.R, B=@n43))
   B=Multiply(A=Multiply(A=@n45.RadialGradientExponential, B=@n8), B=ScalarParameter() {defaultValue=0.5, parameterName=Glow_Alpha})
  )
  n47 = ScalarParameter() {defaultValue=2, parameterName=AdjustSparksAlpha}
  Saturate(None=Add(A=Saturate(None=@n46), B=Multiply(A=@n34, B=@n47)))
 ]
}
