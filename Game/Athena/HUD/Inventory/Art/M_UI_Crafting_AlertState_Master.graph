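# Translucent UI master material, named for the crafting alert state in the inventory HUD, that pulses between a stop color and a start color with animated noise.
# A looping, offsettable timer feeds SmoothCurve functions to shape the swing, which drives both base and noise intensity.
# Two panning blurred-noise samples, one distorting the other's UVs, multiply into a mottled mask in square-corrected coordinates.
# Pulse and pulse-scaled noise add together as the color blend alpha.
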Material /Game/Athena/HUD/Inventory/Art/M_UI_Crafting_AlertState_Master
graph {
 properties {blendMode=BLEND_Translucent, materialDomain=MD_UI, pixelDepthOffsetMode=PDOM_Legacy, usedShadingModels=Default Lit}
 nodes [
  n1 = ScalarParameter() {defaultValue=3, parameterName=SwingBottomTangent}
  n2 = ScalarParameter() {defaultValue=4, parameterName=SwingTopTangent}
  n3 = Time()
  n4 = ScalarParameter() {parameterName=GlobalTimeOffsetRandomizer}
  n5 = Fmod(
   A=Multiply(A=Add(A=@n3, B=@n4), B=ScalarParameter() {defaultValue=1.5, parameterName=GlobalTimeMultiplier}) {constB=1.5}
   B=ScalarParameter() {defaultValue=1.75, parameterName=SwingFrequency}
  )
  n6 = Saturate(None=MaterialFunctionCall!(
    X=MaterialFunctionCall(X=@n5, Tangent 0=@n1, Tangent 1=@n2) {materialFunction=/Engine/Functions/Engine_MaterialFunctions02/Gradients/SmoothCurve}.Result
    Tangent 0
    Tangent 1
   ) {materialFunction=&SmoothCurve}.Result)
  n7 = OneMinus(None=@n5)
  n8 = Saturate(None=DotProduct(
    A=OneMinus(None=@n6)
    B=MaterialFunctionCall(X=@n7) {materialFunction=&SmoothCurve}.Result
   ))
  n9 = LinearInterpolate(
   A=ScalarParameter() {defaultValue=0.25, parameterName=StopAlpha}
   B=ScalarParameter() {defaultValue=0.75, parameterName=Start_Alpha}
   Alpha=@n8
  )
  n10 = ScalarParameter() {defaultValue=0.25, parameterName=NoiseStopAlpha}
  n11 = ScalarParameter() {defaultValue=0.75, parameterName=NoiseStart_Alpha}
  n12 = MaterialFunctionCall(
   UVs=MaterialFunctionCall() {materialFunction=/FortUILibrary/Foundation/Art_ToolKit/Tools/MF_UI_SquareToScreenSpace}.X_Priority
   Texture Scale=ScalarParameter() {defaultValue=1, parameterName=TextureScale}
  ) {materialFunction=&ScaleUVsByCenter}
  n13 = ScalarParameter() {defaultValue=0.125, parameterName=Texture1_XSpeed}
  n14 = ScalarParameter() {defaultValue=-0.05, parameterName=Texture1_YSpeed}
  n15 = ScalarParameter() {defaultValue=-0.05, parameterName=Texture2_XSpeed}
  n16 = ScalarParameter() {defaultValue=0.025, parameterName=Texture2_YSpeed}
  n17 = TextureSample(
   UVs=Panner(Coordinate=@n12.UVs, Speed=AppendVector(A=@n15, B=@n16))
   Tex=TextureObject() {samplerType=SAMPLERTYPE_Grayscale, texture=&LowResBlurredNoise}
  )
  n18 = Add(A=Panner(Coordinate=@n12.UVs, Speed=AppendVector(A=@n13, B=@n14)), B=Multiply(A=Sine(None=@n17.R)) {constB=0.075})
  n19 = TextureObject() {samplerType=SAMPLERTYPE_Grayscale, texture=&LowResBlurredNoise}
  n20 = ScalarParameter() {defaultValue=0.25, parameterName=SubtractTextureAmount}
  n21 = Saturate(None=DotProduct(A=TextureSample(UVs=@n18, Tex=@n19).R, B=Subtract(A=@n17.R, B=@n20) {constB=0.25}))
  n22 = ScalarParameter() {defaultValue=1, parameterName=LiftNoiseAmount}
  n23 = Saturate(None=DotProduct(A=LinearInterpolate(A=@n10, B=@n11, Alpha=@n8), B=Multiply(A=@n21, B=@n22)))
  LinearInterpolate(
   A=VectorParameter() {parameterName=StopColor, defaultValue=<rgba 0 1 0 1>}.RGB
   B=VectorParameter() {parameterName=StartColor, defaultValue=<rgba 1 0 0 1>}.RGB
   Alpha=Saturate(None=Add(A=@n9, B=@n23))
  )
 ]
}
