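# Translucent UI master material in the Athena HUD inventory art that draws a box with four adjustable angled edges.
# Four rotated, offset UV gradients form side and top masks, sharpened by softness parameters scaled through X_Dimension-based resolution correction.
# The combined mask blends inner and border color and alpha, with opacity scaled by an Animate_Alpha lerp between start and stop values.
# Sides sample normalized UVs, while tops sample 9-slice UVs.
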Material /Game/Athena/HUD/Inventory/Art/M_UI_HUD_AngledBox_Master
graph {
 properties {blendMode=BLEND_Translucent, materialDomain=MD_UI, pixelDepthOffsetMode=PDOM_Legacy, usedShadingModels=Default Lit}
 nodes [
  n1 = MaterialFunctionCall() {materialFunction=&GetUserInterfaceUV}
  n2 = MaterialFunctionCall!(Rotation_Fraction=ScalarParameter() {group=1, parameterName=1_Angle_Rotation}, TexCoord=OneMinus(None=@n1.Normalized UV)) {materialFunction=/FortUILibrary/Foundation/Art_ToolKit/AngledBoxes/MF_UI_Rotator_Fractions}
  n3 = ScalarParameter() {defaultValue=-0.75, group=1, parameterName=1_Angle_Offset}
  n4 = Saturate(None=ComponentMask(None=Add(A=@n2.Result, B=@n3)) {r=true})
  n5 = MaterialFunctionCall!(Native_U_PixelSize=ScalarParameter() {defaultValue=128, group=Global, parameterName=X_Dimension}) {materialFunction=/FortUILibrary/Foundation/Art_ToolKit/Tools/M_UI_Resolution_Sharpness_Correction}
  n6 = Multiply(A=@n5.Result, B=ScalarParameter() {defaultValue=64, group=Global, parameterName=Sides_Angle_Softness})
  n7 = MaterialFunctionCall(Rotation_Fraction=ScalarParameter() {group=2, parameterName=2_Angle_Rotation}, TexCoord=@n1.Normalized UV) {materialFunction=&MF_UI_Rotator_Fractions}
  n8 = ScalarParameter() {defaultValue=-0.75, group=2, parameterName=2_Angle_Offset}
  n9 = Saturate(None=ComponentMask(None=Add(A=@n7.Result, B=@n8)) {r=true})
  n10 = Add(A=Saturate(None=Multiply(A=@n4, B=@n6)), B=Saturate(None=Multiply(A=@n9, B=@n6)))
  n11 = MaterialFunctionCall(Rotation_Fraction=ScalarParameter() {group=3, parameterName=3_Angle_Rotation}, TexCoord=OneMinus(None=@n1.9-Slice UV)) {materialFunction=&MF_UI_Rotator_Fractions}
  n12 = ScalarParameter() {defaultValue=-0.75, group=3, parameterName=3_Angle_Offset}
  n13 = Saturate(None=ComponentMask(None=Add(A=@n11.Result, B=@n12)) {g=true})
  n14 = Multiply(A=@n5.Result, B=ScalarParameter() {defaultValue=64, group=Global, parameterName=Tops_Angle_Softness})
  n15 = MaterialFunctionCall(Rotation_Fraction=ScalarParameter() {group=4, parameterName=4_Angle_Rotation}, TexCoord=@n1.9-Slice UV) {materialFunction=&MF_UI_Rotator_Fractions}
  n16 = ScalarParameter() {defaultValue=-0.75, group=4, parameterName=4_Angle_Offset}
  n17 = Saturate(None=ComponentMask(None=Add(A=@n15.Result, B=@n16)) {g=true})
  n18 = Add(A=Saturate(None=Multiply(A=@n13, B=@n14)), B=Saturate(None=Multiply(A=@n17, B=@n14)))
  n19 = Saturate(None=Add(A=Saturate(None=@n10), B=Saturate(None=@n18)))
  LinearInterpolate(
   A=VectorParameter() {group=Color, parameterName=InnerColor, defaultValue=<rgba 0 1 0 1>}.RGB
   B=VectorParameter() {group=Color, parameterName=BorderColor, defaultValue=<rgba 1 0 0 1>}.RGB
   Alpha=@n19
  )
  n20 = LinearInterpolate(
   A=ScalarParameter() {defaultValue=0.5, group=Alpha, parameterName=Inner_Alpha}
   B=ScalarParameter() {defaultValue=1, group=Alpha, parameterName=Border_Alpha}
   Alpha=@n19
  )
  n21 = ScalarParameter() {defaultValue=0.25, group=Alpha, parameterName=FinalAlpha_Start}
  n22 = ScalarParameter() {defaultValue=0.666, group=Alpha, parameterName=FinalAlpha_Stop}
  n23 = ScalarParameter() {defaultValue=1, group=Alpha, parameterName=Animate_Alpha}
  Saturate(None=Multiply(A=Saturate(None=@n20), B=LinearInterpolate(A=@n21, B=@n22, Alpha=@n23)))
 ]
}
