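# Master UI material, translucent, for the HUD crafting panel, drawing a vertical gradient tinted by item rarity.
# The UV's V coordinate, with offset, taper and lift parameters, blends Backing_Color into Rarity_Color.
# A second, separately tunable V gradient blends opacity up from Lift_Alpha, then Final_Alpha scales it.
# It samples no textures, so the look is entirely procedural and parameter-driven.
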Material /Game/Athena/HUD/Inventory/Art/M_UI_HUD_CraftingPanel_Master
graph {
 properties {blendMode=BLEND_Translucent, materialDomain=MD_UI, pixelDepthOffsetMode=PDOM_Legacy, usedShadingModels=Default Lit}
 nodes [
  n1 = TextureCoordinate()
  n2 = Add(A=@n1, B=ScalarParameter() {defaultValue=-0.5, parameterName=Offset_Color_Gradient})
  n3 = Saturate(None=Power(Base=ComponentMask(None=@n2) {g=true}, Exp=ScalarParameter() {defaultValue=1.5, parameterName=Taper_Color_Gradient}))
  n4 = ScalarParameter() {defaultValue=3, parameterName=Lift_Color_Gradient}
  LinearInterpolate(
   A=VectorParameter() {parameterName=Backing_Color, defaultValue=<rgba 0 1 0 1>}.RGB
   B=VectorParameter() {parameterName=Rarity_Color, defaultValue=<rgba 1 0 0 1>}.RGB
   Alpha=Saturate(None=Multiply(A=@n3, B=@n4))
  )
  n5 = ScalarParameter() {defaultValue=0.5, parameterName=Lift_Alpha}
  n6 = ScalarParameter() {parameterName=Offset_Alpha_Gradient}
  n7 = Power(Base=ComponentMask(None=Add(A=@n1, B=@n6)) {g=true}, Exp=ScalarParameter() {defaultValue=1.5, parameterName=Taper_Alpha_Gradient})
  n8 = Saturate(None=Multiply(A=Saturate(None=@n7), B=ScalarParameter() {defaultValue=1, parameterName=Lift_Alpha_Gradient}))
  Multiply(A=Saturate(None=LinearInterpolate(A=@n5, Alpha=@n8)), B=ScalarParameter() {defaultValue=1, parameterName=Final_Alpha})
 ]
}
