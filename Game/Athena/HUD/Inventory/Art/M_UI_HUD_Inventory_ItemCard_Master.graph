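# Translucent UI master material for HUD inventory item cards, compositing an item icon over a gradient backing.
# A radial gradient blends falloff and hover-animated hotspot colors and alphas, beneath an icon with animated scale and Y offset, border-clamped alpha and vertical fade.
# Scalar-driven lerps switch between default, recipe, ingredient and info-panel looks, the last adding its own gradient and soft angled-edge backing.
# The UV aspect term divides pixel width by itself, leaving UVs unscaled.
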Material /Game/Athena/HUD/Inventory/Art/M_UI_HUD_Inventory_ItemCard_Master
graph {
 properties {blendMode=BLEND_Translucent, materialDomain=MD_UI, pixelDepthOffsetMode=PDOM_Legacy, usedShadingModels=Default Lit}
 nodes [
  n1 = VectorParameter() {group=Color, parameterName=BG_Falloff_Color, defaultValue=<rgba 1 0 0 1>}
  n2 = LinearInterpolate(
   A=VectorParameter() {group=Color, parameterName=BG_Hotspot_Color, defaultValue=<rgba 0 1 0 1>}.RGB
   B=VectorParameter() {group=Color, parameterName=BG_Hotspot_Color_Hover, defaultValue=<rgba 0 0 1 1>}.RGB
   Alpha=ScalarParameter() {group=Animate, parameterName=Animate_Hotspot_Color}
  )
  n3 = ScalarParameter() {defaultValue=0.5, group=Background, parameterName=BG_Gradient_X_Position}
  n4 = ScalarParameter() {parameterName=IsRecipeView}
  n5 = ScalarParameter() {defaultValue=0.5, group=Background, parameterName=BG_Gradient_Y_Position}
  n6 = MaterialFunctionCall() {materialFunction=&GetUserInterfaceUV}
  n7 = MaterialFunctionCall(Float2=@n6.Pixel Size) {materialFunction=&BreakOutFloat2Components}
  n8 = Divide(A=@n7.R, B=@n7.R)
  n9 = Constant() {r=1}
  n10 = OneMinus(None=@n8)
  n11 = Add(A=Multiply(A=@n6.Normalized UV, B=AppendVector(A=@n8, B=@n9)), B=AppendVector(A=Multiply(A=@n10) {constB=0.5}, B=Constant()))
  n12 = MaterialFunctionCall(
   UVs=@n11
   CenterPosition=AppendVector(A=LinearInterpolate(A=@n3, Alpha=@n4) {constB=0.5}, B=LinearInterpolate(A=@n5, Alpha=@n4) {constB=0.5})
   Radius=ScalarParameter() {defaultValue=0.5, group=Background, parameterName=BG_Gradient_Radius}
   Density=ScalarParameter() {defaultValue=2, group=Background, parameterName=BG_Gradient_Density}
  ) {materialFunction=&RadialGradientExponential}
  n13 = ScalarParameter() {parameterName=IsRecipeView}
  n14 = LinearInterpolate(
   A=LinearInterpolate(
    A=ScalarParameter() {defaultValue=0.8, group=ItemIcon, parameterName=ItemIcon_Scale_Start}
    B=ScalarParameter() {defaultValue=0.8, parameterName=RecipeViewScaleStart}
    Alpha=@n13
   ) {constB=0}
   B=LinearInterpolate(
    A=ScalarParameter() {defaultValue=1, group=ItemIcon, parameterName=ItemIcon_Scale_Stop}
    B=ScalarParameter() {defaultValue=1, parameterName=RecipeViewScaleStop}
    Alpha=@n13
   )
   Alpha=ScalarParameter() {group=Animate, parameterName=Animate_ItemIcon_Scale}
  )
  n15 = ScalarParameter() {defaultValue=0.8, parameterName=InfoPanel_ItemIcon_Scale}
  n16 = ScalarParameter() {parameterName=IsInfoPanelPresentation}
  n17 = Constant()
  n18 = ScalarParameter() {defaultValue=0.125, parameterName=InfoPanel_Icon_Translate_X}
  n19 = ScalarParameter() {parameterName=IsInfoPanelPresentation}
  n20 = LinearInterpolate(
   A=LinearInterpolate(A=ScalarParameter() {group=ItemIcon, parameterName=ItemIcon_Translate_Y_Start}, Alpha=ScalarParameter() {parameterName=IsRecipeView}) {constB=0}
   B=ScalarParameter() {group=ItemIcon, parameterName=ItemIcon_Translate_Y_Stop}
   Alpha=ScalarParameter() {group=Animate, parameterName=Animate_ItemIcon_Translate_Y}
  )
  n21 = ScalarParameter() {parameterName=InfoPanel_ItemIcon_Translate_Y}
  n22 = ScalarParameter() {parameterName=IsInfoPanelPresentation}
  n23 = Add(
   A=MaterialFunctionCall(UVs=@n11, Texture Scale=LinearInterpolate(A=@n14, B=@n15, Alpha=@n16)) {materialFunction=&ScaleUVsByCenter}.UVs
   B=AppendVector(A=LinearInterpolate(A=@n17, B=@n18, Alpha=@n19), B=LinearInterpolate(A=@n20, B=@n21, Alpha=@n22))
  )
  n24 = TextureObjectParameter() {group=ItemIcon, parameterName=ItemIcon, texture=/Game/UI/Foundation/Textures/Icons/Weapons/Items/T-Icon-Weapons-PumpShotgunMark2TT}
  n25 = TextureSample(UVs=@n23, Tex=@n24) {samplerSource=SSM_Clamp_WorldGroupSettings}
  n26 = MaterialFunctionCall!(UVs=@n23, Alpha=@n25.A) {materialFunction=/Game/UI/Foundation/Items/ClampAlphaToTransparentBorder}
  n27 = LinearInterpolate(
   A=LinearInterpolate(
    A=LinearInterpolate(A=@n1.RGB, B=@n2, Alpha=@n12.RadialGradientExponential)
    B=VectorParameter() {group=Color, parameterName=IngredientView_Backing_Color, defaultValue=<rgba 1 0 0 1>}.RGB
    Alpha=ScalarParameter() {parameterName=IsIngredient}
   )
   B=@n25.RGB
   Alpha=@n26.Result
  )
  n28 = ScalarParameter() {defaultValue=1, parameterName=InfoPanel_Gradient_X}
  n29 = ScalarParameter() {parameterName=InfoPanel_Gradient_Y}
  n30 = Saturate(None=MaterialFunctionCall!(
    CenterPosition=AppendVector(A=@n28, B=@n29)
    Col_01
    Col_02
    Radius Scale=ScalarParameter() {defaultValue=2, parameterName=InfoPanel_Gradient_Radius}
    Density=ScalarParameter() {defaultValue=2, parameterName=InfoPanel_Gradient_Density}
   ) {materialFunction=/Game/UI/Materials/MF_UI_SmoothPoly_RadialGradient}."Gradient (0-1)")
  n31 = LinearInterpolate(
   A=VectorParameter() {group=Color, parameterName=InfoPanel_Color_A, defaultValue=<rgba 1 0 0 1>}.RGB
   B=VectorParameter() {group=Color, parameterName=InfoPanel_Color_B, defaultValue=<rgba 0 1 0 1>}.RGB
   Alpha=@n30
  )
  n32 = ScalarParameter() {parameterName=IsInfoPanelPresentation}
  LinearInterpolate(
   A=LinearInterpolate(A=@n27, B=LinearInterpolate(A=@n31, B=@n25.RGB, Alpha=@n26.Result), Alpha=@n32)
   B=@n27
   Alpha=ScalarParameter() {parameterName=IsIngredient}
  )
  n33 = ScalarParameter() {defaultValue=0.75, group=Background, parameterName=BG_Gradient_Falloff_Alpha}
  n34 = ScalarParameter() {defaultValue=1, group=Background, parameterName=BG_Gradient_Hotspot_Alpha}
  n35 = Add(A=Add(A=TextureCoordinate()) {constB=-0.5}, B=ScalarParameter() {defaultValue=0.25, group=ItemIcon, parameterName=ItemIcon_Fade_Offset})
  n36 = OneMinus(None=Saturate(None=ComponentMask(None=@n35) {g=true}))
  n37 = LinearInterpolate(
   A=LinearInterpolate(
    A=LinearInterpolate(A=@n33, B=@n34, Alpha=@n12.RadialGradientExponential)
    B=ScalarParameter() {defaultValue=0.75, group=Background, parameterName=IngredientView_Alpha}
    Alpha=ScalarParameter() {parameterName=IsIngredient}
   )
   B=LinearInterpolate(A=Saturate(None=@n36), Alpha=ScalarParameter() {parameterName=IsIngredient})
   Alpha=@n26.Result
  )
  n38 = ScalarParameter() {group=Animate, parameterName=Animate_Alpha}
  n39 = ScalarParameter() {defaultValue=-0.05, parameterName=InfoPanel_Angle}
  n40 = Add(
   A=OneMinus(None=MaterialFunctionCall!(Rotation_Fraction=@n39, TexCoord=@n11) {materialFunction=/FortUILibrary/Foundation/Art_ToolKit/AngledBoxes/MF_UI_Rotator_Fractions}.Result)
   B=ScalarParameter() {defaultValue=-0.175, parameterName=Offset_InfoPanel_Angle}
  )
  n41 = Saturate(None=Multiply(A=ComponentMask(None=@n40) {r=true}, B=ScalarParameter() {defaultValue=64, parameterName=Angle_Softness}))
  n42 = LinearInterpolate(
   A=ScalarParameter() {defaultValue=1, parameterName=InfoPanel_Gradient_Alpha_A}
   B=ScalarParameter() {defaultValue=0.5, parameterName=InfoPanel_Gradient_Alpha_B}
   Alpha=@n30
  )
  n43 = Multiply(A=Saturate(None=DotProduct(A=@n41, B=@n42)), B=ScalarParameter() {defaultValue=0.25, parameterName=InfoPanel_Backing_Alpha})
  Saturate(None=LinearInterpolate(A=LinearInterpolate(A=@n37, Alpha=@n38), B=LinearInterpolate(A=@n43, B=@n26.Result, Alpha=@n26.Result), Alpha=@n32))
  TextureCoordinate()
 ]
}
