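# UI master material for an inventory HUD resource backing: a colored, translucent panel with an item icon composited on top.
# Four rotated, offset UV edge ramps form a skewed quad mask, blended from a straight ceil-based mask by AnimateToWonk and animated by AnimateSelection with per-edge random rotation.
# Animate parameters drive icon scale and offset, color, alpha, and a radial-plus-offset-silhouette shadow.
# The IsJustAlpha static switch substitutes white with the simple mask at animatable opacity.
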Material /Game/Athena/HUD/Inventory/Art/M_UI_HUD_Inventory_ResourceBacking_Master
graph {
 properties {blendMode=BLEND_Translucent, materialDomain=MD_UI, pixelDepthOffsetMode=PDOM_Legacy, usedShadingModels=Default Lit}
 nodes [
  n1 = MaterialFunctionCall() {materialFunction=&GetUserInterfaceUV}
  n2 = MaterialFunctionCall(Float2=@n1.Pixel Size) {materialFunction=&BreakOutFloat2Components}
  n3 = Divide(A=@n2.R, B=@n2.R)
  n4 = Constant() {r=1}
  n5 = OneMinus(None=@n3)
  n6 = Add(A=Multiply(A=@n1.Normalized UV, B=AppendVector(A=@n3, B=@n4)), B=AppendVector(A=Multiply(A=@n5) {constB=0.5}, B=Constant()))
  n7 = Add(A=@n6, B=ScalarParameter() {defaultValue=-0.25, parameterName=BackingOffset})
  Sine(None=@n7) {period=1.5}
  n8 = ComponentMask(None=Ceil(None=@n7)) {g=true}
  n9 = ScalarParameter() {parameterName=JustAlphaStart}
  n10 = ScalarParameter() {defaultValue=1, parameterName=JustAlphaStop}
  n11 = ScalarParameter() {parameterName=AnimateJustAlpha}
  n12 = ScalarParameter() {parameterName=AnimateSelection}
  n13 = Add(
   A=TextureCoordinate()
   B=LinearInterpolate(
    A=ScalarParameter() {parameterName=Angle_Offset_Left_Start}
    B=ScalarParameter() {defaultValue=-0.0075, parameterName=Angle_Offset_Left_Stop}
    Alpha=@n12
   )
  )
  n14 = LinearInterpolate(
   A=Constant()
   B=Add(A=ScalarParameter() {defaultValue=0.0018, parameterName=Angle_Rotation_Left}, B=ScalarParameter() {parameterName=RandomValue_Left})
   Alpha=@n12
  )
  n15 = ScalarParameter() {defaultValue=512, parameterName=U_Softness_Start}
  n16 = ScalarParameter() {defaultValue=176, parameterName=U_Softness_Stop}
  n17 = Multiply(
   A=ComponentMask(None=Rotator(Coordinate=@n13, Time=@n14) {speed=6.2832}) {r=true}
   B=Clamp(None=LinearInterpolate(A=@n15, B=@n16, Alpha=@n12), Min=@n16, Max=@n15)
  )
  n18 = Add(
   A=OneMinus(None=TextureCoordinate())
   B=LinearInterpolate(
    A=ScalarParameter() {parameterName=Angle_Offset_Right_Start}
    B=ScalarParameter() {defaultValue=-0.01, parameterName=Angle_Offset_Right_Stop}
    Alpha=@n12
   )
  )
  n19 = LinearInterpolate(
   A=Constant()
   B=Add(A=ScalarParameter() {defaultValue=-0.0025, parameterName=Angle_Rotation_Right}, B=ScalarParameter() {parameterName=RandomValue_Right})
   Alpha=@n12
  )
  n20 = ScalarParameter() {parameterName=U_Softness_Start}
  n21 = ScalarParameter() {parameterName=U_Softness_Stop}
  n22 = Multiply(
   A=ComponentMask(None=Rotator(Coordinate=@n18, Time=@n19) {speed=6.2832}) {r=true}
   B=Clamp(None=LinearInterpolate(A=@n20, B=@n21, Alpha=@n12), Min=@n21, Max=@n20)
  )
  n23 = Saturate(None=DotProduct(A=Clamp(None=@n17), B=Clamp(None=@n22)))
  n24 = Add(
   A=TextureCoordinate()
   B=LinearInterpolate(
    A=ScalarParameter() {defaultValue=-0.16, parameterName=Angle_Offset_Top_Start}
    B=ScalarParameter() {defaultValue=-0.05, parameterName=Angle_Offset_Top_Stop}
    Alpha=@n12
   )
  )
  n25 = LinearInterpolate(
   A=Constant()
   B=Add(A=ScalarParameter() {defaultValue=0.005, parameterName=Angle_Rotation_Top}, B=ScalarParameter() {parameterName=RandomValue_Top})
   Alpha=@n12
  )
  n26 = ScalarParameter() {defaultValue=256, parameterName=V_Softness_Start}
  n27 = ScalarParameter() {defaultValue=56, parameterName=V_Softness_Stop}
  n28 = Multiply(
   A=ComponentMask(None=Rotator(Coordinate=@n24, Time=@n25) {speed=6.2832}) {g=true}
   B=Clamp(None=LinearInterpolate(A=@n26, B=@n27, Alpha=@n12), Min=@n27, Max=@n26)
  )
  n29 = Add(
   A=OneMinus(None=TextureCoordinate())
   B=LinearInterpolate(
    A=ScalarParameter() {defaultValue=-0.16, parameterName=Angle_Offset_Bottom_Start}
    B=ScalarParameter() {defaultValue=-0.05, parameterName=Angle_Offset_Bottom_Stop}
    Alpha=@n12
   )
  )
  n30 = LinearInterpolate(
   A=Constant()
   B=Add(A=ScalarParameter() {defaultValue=-0.0075, parameterName=Angle_Rotation_Bottom}, B=ScalarParameter() {parameterName=RandomValue_Bottom})
   Alpha=@n12
  )
  n31 = ScalarParameter() {parameterName=V_Softness_Start}
  n32 = ScalarParameter() {parameterName=V_Softness_Stop}
  n33 = Multiply(
   A=ComponentMask(None=Rotator(Coordinate=@n29, Time=@n30) {speed=6.2832}) {g=true}
   B=Clamp(None=LinearInterpolate(A=@n31, B=@n32, Alpha=@n12), Min=@n32, Max=@n31)
  )
  n34 = Saturate(None=DotProduct(A=Clamp(None=@n28), B=Clamp(None=@n33)))
  n35 = LinearInterpolate(A=@n8, B=Saturate(None=DotProduct(A=@n23, B=@n34)), Alpha=ScalarParameter() {parameterName=AnimateToWonk})
  n36 = DotProduct(A=@n35, B=ScalarParameter() {defaultValue=1, parameterName=Backing_Alpha_Start})
  n37 = TextureObjectParameter() {group=ItemIcon, parameterName=ItemIcon, texture=/Game/UI/Foundation/Textures/Icons/Items/T_Icon_Ammo_Medium_256}
  n38 = ScalarParameter() {defaultValue=0.7, group=ItemIcon, parameterName=ItemIcon_Scale_Start}
  n39 = ScalarParameter() {defaultValue=0.9, group=ItemIcon, parameterName=ItemIcon_Scale_Stop}
  n40 = ScalarParameter() {group=Animate, parameterName=Animate_ItemIcon_Scale}
  n41 = ScalarParameter() {defaultValue=0.15, group=ItemIcon, parameterName=ItemIcon_Translate_Y_Start}
  n42 = ScalarParameter() {group=ItemIcon, parameterName=ItemIcon_Translate_Y_Stop}
  n43 = ScalarParameter() {group=Animate, parameterName=Animate_ItemIcon_Translate_Y}
  n44 = Add(
   A=MaterialFunctionCall(UVs=@n6, Texture Scale=LinearInterpolate(A=@n38, B=@n39, Alpha=@n40)) {materialFunction=&ScaleUVsByCenter}.UVs
   B=AppendVector(A=Constant(), B=LinearInterpolate(A=@n41, B=@n42, Alpha=@n43))
  )
  n45 = TextureSample(UVs=@n44, Tex=@n37) {samplerSource=SSM_Clamp_WorldGroupSettings}
  n46 = MaterialFunctionCall!(UVs=@n44, Alpha=@n45.A) {materialFunction=/Game/UI/Foundation/Items/ClampAlphaToTransparentBorder}
  n47 = DotProduct(A=@n35, B=ScalarParameter() {defaultValue=1, parameterName=Backing_Alpha_Stop})
  StaticSwitchParameter(
   True=Multiply(A=@n8, B=LinearInterpolate(A=@n9, B=@n10, Alpha=@n11))
   False=LinearInterpolate(
    A=LinearInterpolate(A=@n36, B=@n46.Result, Alpha=@n46.Result)
    B=LinearInterpolate(A=@n47, B=@n46.Result, Alpha=@n46.Result)
    Alpha=ScalarParameter() {parameterName=Animate_Alpha}
   )
  ) {parameterName=IsJustAlpha}
  n48 = LinearInterpolate(
   A=VectorParameter() {parameterName=Color_Start, defaultValue=<rgba 0 1 0 1>}.RGB
   B=VectorParameter() {parameterName=Color_Stop, defaultValue=<rgba 1 0 0 1>}.RGB
   Alpha=ScalarParameter() {parameterName=Animate_Color}
  )
  n49 = Add(A=@n44, B=AppendVector(A=Constant(), B=ScalarParameter() {defaultValue=-0.075, group=ItemIcon, parameterName=ItemIcon_Shadow_Offset}))
  n50 = MaterialFunctionCall(
   UVs=@n49
   Texture Scale=LinearInterpolate(
    A=ScalarParameter() {defaultValue=1, group=ItemIcon, parameterName=HardShadow_Scale_Start}
    B=ScalarParameter() {defaultValue=1.125, group=ItemIcon, parameterName=HardShadow_Scale_Stop}
    Alpha=ScalarParameter() {group=Animate, parameterName=Animate_ItemIcon_Scale}
   )
  ) {materialFunction=&ScaleUVsByCenter}
  n51 = TextureSample(UVs=@n50.UVs, Tex=@n37) {samplerSource=SSM_Clamp_WorldGroupSettings}
  n52 = Add(
   A=MaterialFunctionCall(
    UVs=@n49
    Radius=ScalarParameter() {defaultValue=0.25, parameterName=Shadow_Radius}
    Density=ScalarParameter() {defaultValue=1.666, parameterName=Shadow_Density}
   ) {materialFunction=&RadialGradientExponential}.RadialGradientExponential
   B=Multiply(
    A=MaterialFunctionCall(UVs=@n50.UVs, Alpha=@n51.A) {materialFunction=&ClampAlphaToTransparentBorder}.Result
    B=ScalarParameter() {defaultValue=0.125, parameterName=HardShadowAlpha}
   )
  )
  n53 = LinearInterpolate(
   A=@n48
   B=VectorParameter() {parameterName=Shadow_Color, defaultValue=<rgba 0 0 0 1>}.RGB
   Alpha=Multiply(A=Saturate(None=@n52), B=ScalarParameter() {defaultValue=0.1, parameterName=ShadowLayerAlpha})
  )
  n54 = ScalarParameter() {parameterName=Animate_Shadow}
  StaticSwitchParameter(True=Constant() {r=1}, False=LinearInterpolate(A=LinearInterpolate(A=@n48, B=@n53, Alpha=@n54), B=@n45.RGB, Alpha=@n46.Result)) {parameterName=IsJustAlpha}
 ]
}
