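# Translucent UI master material for a pulsing selection prompt in the inventory HUD, built from two radial gradients.
# Time wrapped by Fmod loops the outer gradient's radius and alpha and both gradients' softness.
# Inner alpha instead follows Animate_Inner_Alpha, while Animate_Outer_Alpha feeds nothing.
# A resolution sharpness correction function scales both densities, and the gradients are summed, clamped and multiplied by FinalAlpha.
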Material /Game/Athena/HUD/Inventory/Art/M_UI_HUD_Inventory_SelectionPrompt_Pulse_Master
graph {
 properties {blendMode=BLEND_Translucent, materialDomain=MD_UI, pixelDepthOffsetMode=PDOM_Legacy, usedShadingModels=Default Lit}
 nodes [
  n1 = Fmod(
   A=Multiply(A=Time(), B=ScalarParameter() {defaultValue=1, parameterName=TimeMultiplier})
   B=ScalarParameter() {defaultValue=1, parameterName=FmodPulseLength}
  )
  n2 = Multiply(
   A=MaterialFunctionCall!(Native_U_PixelSize=ScalarParameter() {defaultValue=80, parameterName=Widget_X_Dimension}) {materialFunction=/FortUILibrary/Foundation/Art_ToolKit/Tools/M_UI_Resolution_Sharpness_Correction}.Result
   B=LinearInterpolate(
    A=ScalarParameter() {defaultValue=4, parameterName=Edge_Softness}
    B=ScalarParameter() {defaultValue=1, parameterName=Edge_Softness_Stop}
    Alpha=@n1
   )
  )
  n3 = MaterialFunctionCall(
   UVs=TextureCoordinate()
   Radius=ScalarParameter() {defaultValue=0.25, parameterName=Inner_Radius}
   Density=Multiply(A=ScalarParameter() {defaultValue=8, parameterName=Inner_Density}, B=@n2)
  ) {materialFunction=&RadialGradientExponential}
  n4 = ScalarParameter() {defaultValue=0.75, parameterName=InnerRadius_Alpha_Start}
  n5 = ScalarParameter() {parameterName=InnerRadius_Alpha_Stop}
  n6 = ScalarParameter() {parameterName=Animate_Inner_Alpha}
  n7 = Saturate(None=Multiply(A=Saturate(None=@n3.RadialGradientExponential), B=LinearInterpolate(A=@n4, B=@n5, Alpha=@n6)))
  n8 = MaterialFunctionCall(
   UVs=TextureCoordinate()
   Radius=LinearInterpolate(
    A=ScalarParameter() {defaultValue=0.25, parameterName=Outer_Radius_Start}
    B=ScalarParameter() {defaultValue=0.5, parameterName=Outer_Radius_Stop}
    Alpha=@n1
   )
   Density=Multiply(A=ScalarParameter() {defaultValue=8, parameterName=Outer_Density}, B=@n2)
  ) {materialFunction=&RadialGradientExponential}
  n9 = ScalarParameter() {defaultValue=0.25, parameterName=OuterRadius_Alpha_Start}
  n10 = ScalarParameter() {parameterName=OuterRadius_Alpha_Stop}
  n11 = Saturate(None=Multiply(A=Saturate(None=@n8.RadialGradientExponential), B=LinearInterpolate(A=@n9, B=@n10, Alpha=@n1)))
  Multiply(A=Saturate(None=Add(A=@n7, B=@n11)), B=ScalarParameter() {defaultValue=1, parameterName=FinalAlpha})
  ScalarParameter() {parameterName=Animate_Outer_Alpha}
  VectorParameter() {parameterName=Color, defaultValue=<rgba 1 1 1 1>}
 ]
}
