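# Translucent UI master material for the HUD inventory stats bar, drawing a horizontal fill with backing, delta and amount colors.
# Percent and Delta_Percent each offset the U coordinate, and a shared Bar_Softness multiplier turns the results into saturated, inverted edge masks.
# Color lerps layer amount over delta over backing, while a separate lerp blends Backing_Alpha to Bars_Alpha using only the delta mask.
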Material /Game/Athena/HUD/Inventory/Art/M_UI_HUD_Inventory_StatsBar_Master
graph {
 properties {blendMode=BLEND_Translucent, materialDomain=MD_UI, pixelDepthOffsetMode=PDOM_Legacy, usedShadingModels=Default Lit}
 nodes [
  n1 = TextureCoordinate()
  n2 = Multiply(A=ScalarParameter() {defaultValue=75, parameterName=Delta_Percent}) {constB=-0.01}
  n3 = Multiply(A=ComponentMask(None=Add(A=@n1, B=@n2)) {r=true}, B=ScalarParameter() {defaultValue=256, parameterName=Bar_Softness})
  n4 = Saturate(None=OneMinus(None=Saturate(None=@n3)))
  n5 = Add(A=TextureCoordinate(), B=Multiply(A=ScalarParameter() {defaultValue=50, parameterName=Percent}) {constB=-0.01})
  n6 = Saturate(None=Multiply(A=ComponentMask(None=@n5) {r=true}, B=ScalarParameter() {defaultValue=256, parameterName=Bar_Softness}))
  LinearInterpolate(
   A=LinearInterpolate(
    A=VectorParameter() {parameterName=Backing_Color, defaultValue=<rgba 0 0 1 1>}.RGB
    B=VectorParameter() {parameterName=Delta_Color, defaultValue=<rgba 0 1 0 1>}.RGB
    Alpha=@n4
   )
   B=VectorParameter() {parameterName=Amount_Color, defaultValue=<rgba 1 0 0 1>}.RGB
   Alpha=Saturate(None=OneMinus(None=@n6))
  )
  LinearInterpolate(
   A=ScalarParameter() {defaultValue=0.5, parameterName=Backing_Alpha}
   B=ScalarParameter() {defaultValue=1, parameterName=Bars_Alpha}
   Alpha=@n4
  )
  OneMinus(None=@n1)
 ]
}
