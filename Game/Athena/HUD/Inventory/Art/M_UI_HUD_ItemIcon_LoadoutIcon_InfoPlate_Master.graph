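# Translucent UI master material for the angled info plate on HUD inventory loadout item icons.
# Plate color blends between start and stop parameters with a hover animation value, and a PostRotation parameter drives a separate UV rotation.
# The plate shape comes from rotated, offset UVs, whose saturated green channel is multiplied by resolution-based sharpness correction and edge softness.
# That edge mask is scaled by an animatable start-to-stop alpha to give opacity.
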Material /Game/Athena/HUD/Inventory/Art/M_UI_HUD_ItemIcon_LoadoutIcon_InfoPlate_Master
graph {
 properties {blendMode=BLEND_Translucent, materialDomain=MD_UI, numCustomizedUVs=1, pixelDepthOffsetMode=PDOM_Legacy, usedShadingModels=Default Lit}
 nodes [
  LinearInterpolate(
   A=VectorParameter() {group=Color, parameterName=Plate_Color_Start, defaultValue=<rgba 1 0 0 1>}.RGB
   B=VectorParameter() {group=Color, parameterName=Plate_Color_Stop, defaultValue=<rgba 0 1 0 1>}.RGB
   Alpha=ScalarParameter() {parameterName=Animate_Color_Hover}
  )
  MaterialFunctionCall!(Rotation_Fraction=ScalarParameter() {parameterName=PostRotation}, TexCoord=TextureCoordinate()) {materialFunction=/FortUILibrary/Foundation/Art_ToolKit/AngledBoxes/MF_UI_Rotator_Fractions}
  n1 = Add(
   A=MaterialFunctionCall(Rotation_Fraction=ScalarParameter() {defaultValue=0.025, parameterName=Plate_Angle}, TexCoord=TextureCoordinate()) {materialFunction=&MF_UI_Rotator_Fractions}.Result
   B=AppendVector(A=Constant(), B=ScalarParameter() {defaultValue=-0.075, parameterName=Offset_Angle_Top})
  )
  n2 = ScalarParameter() {defaultValue=120, parameterName=Widget_X_Dimension}
  n3 = Multiply(
   A=Saturate(None=ComponentMask(None=@n1) {g=true})
   B=Multiply(
    A=MaterialFunctionCall!(Native_U_PixelSize=@n2) {materialFunction=/FortUILibrary/Foundation/Art_ToolKit/Tools/M_UI_Resolution_Sharpness_Correction}.Result
    B=ScalarParameter() {defaultValue=16, parameterName=Angle_Edge_Softness}
   )
  )
  n4 = ScalarParameter() {defaultValue=0.75, parameterName=Alpha_Start}
  n5 = ScalarParameter() {defaultValue=1, parameterName=Alpha_Stop}
  n6 = ScalarParameter() {parameterName=Animate_Alpha}
  Saturate(None=DotProduct(A=Saturate(None=@n3), B=LinearInterpolate(A=@n4, B=@n5, Alpha=@n6)))
 ]
}
