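# UI material function that derives a sharpness factor from a widget's pixel width, likely to keep material edges crisp as it scales.
# It takes the X of Pixel Size from GetUserInterfaceUV, divides it by a fixed 64 and multiplies by the scalar Softness input.
# Softness is described as the image's native 1080p pixel width, despite its name, and falls back to its preview value when unconnected.
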MaterialFunction /FortUILibrary/Foundation/Art_ToolKit/Tools/M_UI_Auto_Sharpen
graph {nodes [
  n1 = MaterialFunctionCall(Float2=MaterialFunctionCall() {materialFunction=&GetUserInterfaceUV}.Pixel Size) {materialFunction=&BreakOutFloat2Components}
  n2 = Constant() {r=64}
  n3 = FunctionInput() {
   bUsePreviewValueAsDefault=true
   description="The native 1080p width (in pixels) of the image this material will live on"
   inputName=Softness
   inputType=FunctionInput_Scalar
  }
  FunctionOutput(None=Multiply(A=Divide(A=@n1.R, B=@n2), B=@n3))
 ]}
