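# Translucent UI master material that draws the circular, two-color rarity tier backing for the inventory HUD.
# An inner exponential radial gradient mixes outer and inner colors, while an outer one, scaled by FinalAlpha, shapes opacity.
# Animate_Color shifts the rim toward the inner color and the center toward white, and BlendOuterColor tints the rim toward the inner color.
# A resolution sharpness function, driven by widget width and edge softness, scales both gradient densities.
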Material /Game/Athena/HUD/Inventory/Art/M_UI_HUD_RarityTier_Backing_Master
graph {
 properties {blendMode=BLEND_Translucent, materialDomain=MD_UI, pixelDepthOffsetMode=PDOM_Legacy, usedShadingModels=Default Lit}
 nodes [
  n1 = VectorParameter() {parameterName=Outer_Color, defaultValue=<rgba 0 1 0 1>}
  n2 = VectorParameter() {parameterName=Inner_Color, defaultValue=<rgba 1 0 0 1>}
  n3 = ScalarParameter() {parameterName=BlendOuterColor}
  n4 = ScalarParameter() {parameterName=Animate_Color}
  n5 = TextureCoordinate()
  n6 = ScalarParameter() {defaultValue=0.4, parameterName=Inner_Radius}
  n7 = ScalarParameter() {defaultValue=4, parameterName=Edge_Softness}
  n8 = Multiply(
   A=MaterialFunctionCall!(Native_U_PixelSize=ScalarParameter() {defaultValue=120, parameterName=Widget_X_Dimension}) {materialFunction=/FortUILibrary/Foundation/Art_ToolKit/Tools/M_UI_Resolution_Sharpness_Correction}.Result
   B=@n7
  )
  n9 = Multiply(A=ScalarParameter() {defaultValue=8, parameterName=Inner_Density}, B=@n8)
  LinearInterpolate(
   A=LinearInterpolate(A=LinearInterpolate(A=@n1.RGB, B=@n2.RGB, Alpha=@n3), B=@n2.RGB, Alpha=@n4)
   B=LinearInterpolate(A=@n2.RGB, Alpha=@n4)
   Alpha=Saturate(None=MaterialFunctionCall(UVs=@n5, Radius=@n6, Density=@n9) {materialFunction=&RadialGradientExponential}.RadialGradientExponential)
  )
  MaterialFunctionCall!(Softness=@n7) {materialFunction=/FortUILibrary/Foundation/Art_ToolKit/Tools/M_UI_Auto_Sharpen}
  n10 = TextureCoordinate()
  n11 = ScalarParameter() {defaultValue=0.5, parameterName=Outer_Radius}
  n12 = Multiply(A=ScalarParameter() {defaultValue=8, parameterName=Outer_Density}, B=@n8)
  Multiply(
   A=Saturate(None=MaterialFunctionCall(UVs=@n10, Radius=@n11, Density=@n12) {materialFunction=&RadialGradientExponential}.RadialGradientExponential)
   B=ScalarParameter() {defaultValue=1, parameterName=FinalAlpha}
  )
 ]
}
