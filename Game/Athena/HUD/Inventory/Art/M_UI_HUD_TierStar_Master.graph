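# Master UI material for the HUD inventory tier star, drawing a translucent star tinted by item rarity.
# Color blends from white to Rarity_Color through Animate_Color.
# The star shape samples the green channel of a distance-field texture with center-scaled, vertically offset UVs.
# A threshold subtraction, multiplied by Edge_Softness and a resolution sharpness correction based on widget width, controls edge crispness across screen sizes.
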Material /Game/Athena/HUD/Inventory/Art/M_UI_HUD_TierStar_Master
graph {
 properties {blendMode=BLEND_Translucent, materialDomain=MD_UI, pixelDepthOffsetMode=PDOM_Legacy, usedShadingModels=Default Lit}
 nodes [
  LinearInterpolate(
   A=Constant() {r=1}
   B=VectorParameter() {parameterName=Rarity_Color, defaultValue=<rgba 1 0 0 1>}.RGB
   Alpha=ScalarParameter() {parameterName=Animate_Color}
  )
  n1 = Add(
   A=MaterialFunctionCall(UVs=TextureCoordinate(), Texture Scale=ScalarParameter() {defaultValue=1, parameterName=Scale}) {materialFunction=&ScaleUVsByCenter}.UVs
   B=AppendVector(A=Constant(), B=ScalarParameter() {parameterName=Offset})
  )
  n2 = ScalarParameter() {defaultValue=8, parameterName=Widget_X_Dimension}
  Multiply(
   A=Subtract(
    A=TextureSample(UVs=@n1) {samplerSource=SSM_Clamp_WorldGroupSettings, texture=/Game/Athena/HUD/VolumePrompt/Materials/T_UI_StarDF}.G
    B=ScalarParameter() {defaultValue=0.25, parameterName=Subtract}
   ) {constB=0.25}
   B=Multiply(
    A=MaterialFunctionCall!(Native_U_PixelSize=@n2) {materialFunction=/FortUILibrary/Foundation/Art_ToolKit/Tools/M_UI_Resolution_Sharpness_Correction}.Result
    B=ScalarParameter() {defaultValue=4, parameterName=Edge_Softness}
   )
  )
 ]
}
