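# Translucent UI master material for a HUD vignette, with parameters for position, scale, radius, density, color and alpha.
# An exponential radial gradient is built on scaled UI UVs; the apparent aspect correction divides pixel-size R by itself, so it has no effect.
# The inverted gradient blends InnerColor to OuterColor and forms an alpha mask, faded between start and stop values by Animate_Alpha.
# Separate Lift and SmoothCurve branches also reshape the gradient.
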Material /Game/Athena/HUD/Inventory/Art/M_UI_HUD_Vignette_Master
graph {
 properties {blendMode=BLEND_Translucent, materialDomain=MD_UI, pixelDepthOffsetMode=PDOM_Legacy, usedShadingModels=Default Lit}
 nodes [
  n1 = MaterialFunctionCall() {materialFunction=&GetUserInterfaceUV}
  n2 = MaterialFunctionCall(Float2=@n1.Pixel Size) {materialFunction=&BreakOutFloat2Components}
  n3 = Divide(A=@n2.R, B=@n2.R)
  n4 = Multiply(A=@n1.Normalized UV, B=AppendVector(A=@n3, B=Constant() {r=1}))
  n5 = AppendVector(A=Multiply(A=OneMinus(None=@n3)) {constB=0.5}, B=Constant())
  n6 = MaterialFunctionCall(
   UVs=MaterialFunctionCall(UVs=Add(A=@n4, B=@n5), Texture Scale=ScalarParameter() {defaultValue=1, group=Location, parameterName=Scale}) {materialFunction=&ScaleUVsByCenter}.UVs
   CenterPosition=AppendVector(
    A=ScalarParameter() {defaultValue=0.5, group=Location, parameterName=Shift_X}
    B=ScalarParameter() {defaultValue=0.5, group=Location, parameterName=Shift_Y}
   )
   Radius=ScalarParameter() {defaultValue=2.33, group=Density, parameterName=Radius}
   Density=ScalarParameter() {defaultValue=2.33, group=Density, parameterName=Density}
  ) {materialFunction=&RadialGradientExponential}
  n7 = Saturate(None=@n6.RadialGradientExponential)
  n8 = Saturate(None=OneMinus(None=@n7))
  LinearInterpolate(
   A=VectorParameter() {group=Color, parameterName=InnerColor, defaultValue=<rgba 1 0 0 1>}.RGB
   B=VectorParameter() {group=Color, parameterName=OuterColor, defaultValue=<rgba 0 1 0 1>}.RGB
   Alpha=@n8
  )
  Multiply(A=@n7, B=ScalarParameter() {defaultValue=1, group=Alpha, parameterName=Lift})
  n9 = LinearInterpolate(
   A=ScalarParameter() {defaultValue=1, group=Alpha, parameterName=Alpha_Start}
   B=ScalarParameter() {defaultValue=1, group=Alpha, parameterName=Alpha_Stop}
   Alpha=ScalarParameter() {group=Alpha, parameterName=Animate_Alpha}
  )
  Saturate(None=Multiply(A=Multiply(A=@n8, B=@n9), B=ScalarParameter() {defaultValue=1, group=Alpha, parameterName=FinalAlpha}))
  Saturate(None=MaterialFunctionCall!(
    X=@n6.RadialGradientExponential
    Tangent 0=ScalarParameter() {group=Rate, parameterName=Tangent_Top}
    Tangent 1=ScalarParameter() {defaultValue=3, group=Rate, parameterName=Tangent_Bottom}
   ) {materialFunction=/Engine/Functions/Engine_MaterialFunctions02/Gradients/SmoothCurve}.Result)
 ]
}
