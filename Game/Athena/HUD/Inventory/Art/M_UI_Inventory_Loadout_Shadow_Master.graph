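# Translucent UI master material for a soft radial shadow in the inventory loadout HUD.
# An exponential radial gradient with tunable density is saturated, then scaled by an alpha blended from Alpha_Start to Alpha_Stop through the Animate_Alpha parameter, and clamped.
# A Color vector parameter, red by default, supplies the tint, so instances can vary color, falloff and fade.
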Material /Game/Athena/HUD/Inventory/Art/M_UI_Inventory_Loadout_Shadow_Master
graph {
 properties {blendMode=BLEND_Translucent, materialDomain=MD_UI, pixelDepthOffsetMode=PDOM_Legacy, usedShadingModels=Default Lit}
 nodes [
  n1 = MaterialFunctionCall(UVs=TextureCoordinate(), Density=ScalarParameter() {defaultValue=2, parameterName=Density}) {materialFunction=&RadialGradientExponential}
  n2 = ScalarParameter() {parameterName=Alpha_Start}
  n3 = ScalarParameter() {defaultValue=0.125, parameterName=Alpha_Stop}
  n4 = ScalarParameter() {parameterName=Animate_Alpha}
  Saturate(None=Multiply(A=Saturate(None=@n1.RadialGradientExponential), B=LinearInterpolate(A=@n2, B=@n3, Alpha=@n4)))
  VectorParameter() {parameterName=Color, defaultValue=<rgba 1 0 0 1>}
 ]
}
