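# M_UI_UI_SimpleColor_Master is a translucent UI master material for parameter-animated flat colors or tinted icons in the HUD inventory.
# Animate_Color blends between start and stop colors, while Animate_Alpha blends between start and stop alpha values.
# The IsIcon static switch masks the animated alpha with the IconImage texture's alpha, using a saturated dot product of the two scalars as a multiply.
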Material /Game/Athena/HUD/Inventory/Art/M_UI_UI_SimpleColor_Master
graph {
 properties {blendMode=BLEND_Translucent, materialDomain=MD_UI, pixelDepthOffsetMode=PDOM_Legacy, usedShadingModels=Default Lit}
 nodes [
  LinearInterpolate(
   A=VectorParameter() {parameterName=Color_Start, defaultValue=<rgba 0 1 0 1>}.RGB
   B=VectorParameter() {parameterName=Color_Stop, defaultValue=<rgba 1 0 0 1>}.RGB
   Alpha=ScalarParameter() {parameterName=Animate_Color}
  )
  n1 = TextureSample(
   UVs=TextureCoordinate()
   Tex=TextureObjectParameter() {parameterName=IconImage, texture=/Game/UI/Foundation/Textures/Icons/Ammo/T-Icon-Ammo-Medium-32}
  )
  n2 = LinearInterpolate(
   A=ScalarParameter() {defaultValue=1, parameterName=Alpha_Start}
   B=ScalarParameter() {defaultValue=1, parameterName=Alpha_Stop}
   Alpha=ScalarParameter() {parameterName=Animate_Alpha}
  )
  StaticSwitchParameter(True=Saturate(None=DotProduct(A=@n1.A, B=@n2)), False=@n2) {parameterName=IsIcon}
 ]
}
